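# Additive unlit material that draws a localized force splat with foam for the water fluid simulation, allowing negative values.
# A sphere mask in UV space, offset by the force's height relative to WaterZ over its diameter, scales force F.
# Foam derives from the mask's remapped edge and force above a threshold, then is packed with force and negated.
# Panning noise and water-zone height sampling branches are also present.
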Material /FortWater/FluidSimulation/Materials/Forces/M_Fluid_Sim_Force_Component
graph {
 properties {
  bAllowNegativeEmissiveColor=true
  bUseTranslucencyVertexFog=false
  bUsedWithNanite=true
  blendMode=BLEND_Additive
  pixelDepthOffsetMode=PDOM_Legacy
  shadingModel=MSM_Unlit
  usedShadingModels=Unlit
 }
 nodes [
  n1 = ComponentMask(None=VectorParameter() {parameterName=P, defaultValue=<rgba 0.5 0.5 0.01 0>}.RGB) {g=true, r=true}
  n2 = MaterialFunctionCall() {materialFunction=/Water/Materials/Functions/WaterZoneData}
  n3 = Divide(A=@n2.Extent)
  n4 = Divide(
   A=Subtract(A=ComponentMask(None=@n1) {g=true, r=true}, B=ComponentMask(None=@n2.Location) {g=true, r=true})
   B=ComponentMask(None=Subtract(A=@n2.Location, B=@n3)) {g=true, r=true}
  )
  LinearInterpolate(
   A=ScalarParameter() {parameterName=WaterZ}
   B=TextureSample!(
    UVs=Saturate(None=@n4)
    Tex=TextureObjectParameter() {parameterName=WaterVelocityAndHeight, texture=&Black}
    Level=ScalarParameter() {parameterName=WaterVelocityAndHeightMipLevel}
    Apply View MipBias
   ) {mipValueMode=TMVM_MipLevel, texture=&Black}.B
   Alpha=ScalarParameter() {defaultValue=1, parameterName=LandscapeWaterSystem}
  )
  n5 = Constant() {r=0.5}
  n6 = Subtract(A=VectorParameter() {parameterName=P, defaultValue=<rgba 0.5 0.5 0.01 0>}.B, B=ScalarParameter() {parameterName=WaterZ})
  n7 = Multiply(A=ScalarParameter() {defaultValue=32, parameterName=R}) {constB=2}
  n8 = SphereMask(
   A=AppendVector(A=TextureCoordinate(), B=Constant())
   B=MaterialFunctionCall(X=@n5, Y=@n5, Z=Divide(A=@n6, B=@n7) {desc=Normalize Z Location over Diameter}) {materialFunction=&MakeFloat3}.Result
   Radius=Constant() {r=0.5}
   Hardness=Constant()
  )
  n9 = Multiply(A=@n8, B=ScalarParameter() {defaultValue=1, parameterName=F})
  n10 = Panner(
   Coordinate=Multiply(A=TextureCoordinate(), B=ScalarParameter() {defaultValue=1, parameterName=NoiseTiling})
   Time=ScalarParameter() {parameterName=T}
   Speed=ScalarParameter() {defaultValue=0.5, parameterName=Pan Speed}
  ) {speedX=0.1, speedY=0.1}
  Multiply(
   A=@n9
   B=Power(
    Base=TextureSample!(UVs=@n10, Tex, Apply View MipBias) {texture=/FortWater/Textures/T_WaterFlow_01_Foam_Tiled}.R
    Exp=ScalarParameter() {defaultValue=0.5, parameterName=power}
   )
  )
  n11 = OneMinus(None=@n8)
  n12 = ScalarParameter() {defaultValue=0.5, parameterName=fjh}
  n13 = Subtract(A=Abs(None=ScalarParameter() {defaultValue=1, parameterName=F}), B=ScalarParameter() {defaultValue=0.25, parameterName=MinFoamForce})
  n14 = Multiply(
   A=Multiply(
    A=@n8
    B=MaterialFunctionCall!(Alpha=@n11, Offset=@n12) {materialFunction=/Engine/Functions/Engine_MaterialFunctions01/Opacity/AlphaOffset}.Result
   )
   B=Multiply(A=Max(A=@n13) {constB=0}, B=ScalarParameter() {defaultValue=0.08, parameterName=Foam Strength}) {constB=10}
  ) {constB=0}
  Multiply(A=Abs(None=AppendVector(A=@n9, B=@n14))) {constB=-1}
  ScalarParameter() {defaultValue=1, parameterName=Final Opacity}
 ]
}
