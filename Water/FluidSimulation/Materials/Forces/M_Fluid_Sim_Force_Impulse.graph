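# Additive unlit material that writes a circular force impulse and matching foam amount for the water fluid simulation.
# A soft radial UV mask is scaled by force F, which can be negative, and broken up by a panning tiled foam texture to form the first channel.
# The second channel carries foam from an offset copy of the mask, scaled by Foam Strength and the force magnitude above MinFoamForce.
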Material /FortWater/FluidSimulation/Materials/Forces/M_Fluid_Sim_Force_Impulse
graph {
 properties {
  bAllowNegativeEmissiveColor=true
  bUseTranslucencyVertexFog=false
  bUsedWithNanite=true
  blendMode=BLEND_Additive
  pixelDepthOffsetMode=PDOM_Legacy
  shadingModel=MSM_Unlit
  usedShadingModels=Unlit
 }
 nodes [
  n1 = SphereMask(A=AppendVector(A=TextureCoordinate(), B=Constant() {r=0.5}), B=Constant() {r=0.5}, Radius=Constant() {r=0.5}, Hardness=Constant())
  n2 = ScalarParameter() {defaultValue=1, parameterName=F}
  n3 = TextureCoordinate()
  n4 = ScalarParameter() {defaultValue=1, parameterName=NoiseTiling}
  n5 = TextureSample(UVs=Panner(
    Coordinate=Multiply(A=@n3, B=@n4)
    Time=ScalarParameter() {parameterName=T}
    Speed=ScalarParameter() {defaultValue=0.5, parameterName=Pan Speed}
   ) {speedX=0.1, speedY=0.1}) {texture=/FortWater/Textures/T_WaterFlow_01_Foam_Tiled}
  n6 = ScalarParameter() {defaultValue=0.5, parameterName=power}
  n7 = MaterialFunctionCall!(Alpha=@n1, Offset=ScalarParameter() {defaultValue=0.75, parameterName=fjh}) {materialFunction=/Engine/Functions/Engine_MaterialFunctions01/Opacity/AlphaOffset}
  n8 = ScalarParameter() {defaultValue=1, parameterName=F}
  n9 = Max(A=Subtract(A=Abs(None=@n8), B=ScalarParameter() {defaultValue=0.25, parameterName=MinFoamForce})) {constB=0}
  n10 = ScalarParameter() {defaultValue=1, parameterName=Foam Strength}
  AppendVector(
   A=Multiply(A=Multiply(A=@n1, B=@n2), B=Power(Base=@n5.R, Exp=@n6))
   B=Multiply(A=Multiply(A=@n7.Result), B=Multiply(A=@n9, B=@n10) {constB=10}) {constB=0}
  )
  ScalarParameter() {defaultValue=1, parameterName=Final Opacity}
 ]
}
